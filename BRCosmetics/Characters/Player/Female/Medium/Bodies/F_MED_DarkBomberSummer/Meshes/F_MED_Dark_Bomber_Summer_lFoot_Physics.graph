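# Physics asset for the left-foot mesh of the Dark Bomber Summer female medium body cosmetic, set up around dangling shoelace bones.
# Capsule bodies cover pelvis, thighs, left leg and foot, with a fitted foot capsule and small lace capsules.
# Two two-segment lace chains branch from a shared lace root on the foot.
# Leg and lace-root joints rotate freely, while lace joints have swing limits and tight twist limits.
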PhysicsAsset /BRCosmetics/Characters/Player/Female/Medium/Bodies/F_MED_DarkBomberSummer/Meshes/F_MED_Dark_Bomber_Summer_lFoot_Physics
bodies {
 ball_l [{shapeType=Capsule, length=2, radius=2}]
 calf_l [{shapeType=Capsule, length=2, radius=2}]
 calf_twist_02_l [{shapeType=Capsule, length=2, radius=2}]
 dyn_shoeLace_l_a_01 [{shapeType=Capsule, length=0.9244, radius=1.0756, center=<1.3924 0 0>, rotation=<rot 90 0 0>}]
 dyn_shoeLace_l_a_02 [{shapeType=Capsule, length=0.9471, radius=1.0529, center=<0.8056 0 0.0001>, rotation=<rot 90 0 0>}]
 dyn_shoeLace_l_b_01 [{shapeType=Capsule, length=0.9244, radius=1.0756, center=<1.4083 0 0>, rotation=<rot 87.6325 10.4279 12.5372>}]
 dyn_shoeLace_l_b_02 [{shapeType=Capsule, length=0.9471, radius=1.0529, center=<1.1263 0 0>, rotation=<rot 90 0 0>}]
 dyn_shoeLace_loot_l [{shapeType=Capsule, length=0.505, radius=0.505}]
 foot_l [{shapeType=Capsule, length=10.2446, radius=3.6429, center=<5.6724 0.1466 -0.2121>, rotation=<rot 88.0434 179.9983 -167.3024>}]
 pelvis [{shapeType=Capsule, length=13.1066, radius=8.7386}]
 thigh_l [{shapeType=Capsule, length=2, radius=2}]
 thigh_r [{shapeType=Capsule, length=2, radius=2}]
 thigh_twist_01_l [{shapeType=Capsule, length=2, radius=2}]
}
constraints [
 {
  bone1Name=thigh_l
  bone2Name=pelvis
  swing1LimitDegrees=45
  swing1Motion=Free
  swing2LimitDegrees=45
  swing2Motion=Free
  twistLimitDegrees=45
  twistMotion=Free
 }
 {
  bone1Name=calf_l
  bone2Name=thigh_l
  swing1LimitDegrees=45
  swing1Motion=Free
  swing2LimitDegrees=45
  swing2Motion=Free
  twistLimitDegrees=45
  twistMotion=Free
 }
 {
  bone1Name=calf_twist_02_l
  bone2Name=calf_l
  swing1LimitDegrees=45
  swing1Motion=Free
  swing2LimitDegrees=45
  swing2Motion=Free
  twistLimitDegrees=45
  twistMotion=Free
 }
 {
  bone1Name=foot_l
  bone2Name=calf_l
  swing1LimitDegrees=45
  swing1Motion=Free
  swing2LimitDegrees=45
  swing2Motion=Free
  twistLimitDegrees=45
  twistMotion=Free
 }
 {
  bone1Name=ball_l
  bone2Name=foot_l
  swing1LimitDegrees=45
  swing1Motion=Free
  swing2LimitDegrees=45
  swing2Motion=Free
  twistLimitDegrees=45
  twistMotion=Free
 }
 {
  bone1Name=thigh_twist_01_l
  bone2Name=thigh_l
  swing1LimitDegrees=45
  swing1Motion=Free
  swing2LimitDegrees=45
  swing2Motion=Free
  twistLimitDegrees=45
  twistMotion=Free
 }
 {
  bone1Name=thigh_r
  bone2Name=pelvis
  swing1LimitDegrees=45
  swing1Motion=Free
  swing2LimitDegrees=45
  swing2Motion=Free
  twistLimitDegrees=45
  twistMotion=Free
 }
 {bone1Name=dyn_shoeLace_l_b_01, bone2Name=dyn_shoeLace_loot_l, swing1LimitDegrees=45, swing2LimitDegrees=45, twistLimitDegrees=5}
 {bone1Name=dyn_shoeLace_l_b_02, bone2Name=dyn_shoeLace_l_b_01, swing1LimitDegrees=45, swing2LimitDegrees=45, twistLimitDegrees=5}
 {bone1Name=dyn_shoeLace_l_a_01, bone2Name=dyn_shoeLace_loot_l, swing1LimitDegrees=45, swing2LimitDegrees=45, twistLimitDegrees=5}
 {bone1Name=dyn_shoeLace_l_a_02, bone2Name=dyn_shoeLace_l_a_01, swing1LimitDegrees=45, swing2LimitDegrees=45, twistLimitDegrees=5}
 {
  bone1Name=dyn_shoeLace_loot_l
  bone2Name=foot_l
  swing1LimitDegrees=45
  swing1Motion=Free
  swing2LimitDegrees=45
  swing2Motion=Free
  twistLimitDegrees=45
  twistMotion=Free
 }
]
