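# Physics asset for the right-foot mesh of the female medium Dark Bomber Summer player cosmetic, including bodies for the dynamic shoelace bones.
# Capsules cover the pelvis, both legs, the right foot and ball, and two two-segment lace chains that branch from a lace root attached to the foot.
# Lace chain joints swing freely with restricted twist, while the lace root and skeletal joints are free on all angular axes.
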PhysicsAsset /BRCosmetics/Characters/Player/Female/Medium/Bodies/F_MED_DarkBomberSummer/Meshes/F_MED_Dark_Bomber_Summer_rFoot_Physics
bodies {
 ball_r [{shapeType=Capsule, length=2, radius=2}]
 calf_l [{shapeType=Capsule, length=2, radius=2}]
 calf_r [{shapeType=Capsule, length=2, radius=2}]
 calf_twist_02_l [{shapeType=Capsule, length=2, radius=2}]
 calf_twist_02_r [{shapeType=Capsule, length=2, radius=2}]
 dyn_shoeLace_r_a_01 [{shapeType=Capsule, length=1.731, radius=0.9986, center=<1.674 0 0>, rotation=<rot 88.9539 -87.82 -92.5124>}]
 dyn_shoeLace_r_a_02 [{shapeType=Capsule, length=0.9101, radius=1.0882, center=<1.1627 0 0.0001>, rotation=<rot 90 0 0>}]
 dyn_shoeLace_r_b_01 [{shapeType=Capsule, length=1.3914, radius=1.0582, center=<1.4506 0 0>, rotation=<rot 90 0 0>}]
 dyn_shoeLace_r_b_02 [{shapeType=Capsule, length=1.3037, radius=1.0411, center=<0.8551 -0.0002 0>, rotation=<rot 90 0 0>}]
 dyn_shoeLace_root_r [{shapeType=Capsule, length=0.505, radius=0.505}]
 foot_r [{shapeType=Capsule, length=10.0962, radius=3.7912, center=<-5.1522 -0.8908 -0.3365>, rotation=<rot 87.6227 -161.3797 -144.6041>}]
 pelvis [{shapeType=Capsule, length=13.1066, radius=8.7386}]
 thigh_l [{shapeType=Capsule, length=2, radius=2}]
 thigh_r [{shapeType=Capsule, length=2, radius=2}]
 thigh_twist_01_l [{shapeType=Capsule, length=2, radius=2}]
 thigh_twist_01_r [{shapeType=Capsule, length=2, radius=2}]
}
constraints [
 {
  bone1Name=dyn_shoeLace_r_a_02
  bone2Name=dyn_shoeLace_r_a_01
  swing1LimitDegrees=45
  swing1Motion=Free
  swing2LimitDegrees=45
  swing2Motion=Free
  twistLimitDegrees=5
 }
 {
  bone1Name=dyn_shoeLace_r_b_02
  bone2Name=dyn_shoeLace_r_b_01
  swing1LimitDegrees=45
  swing1Motion=Free
  swing2LimitDegrees=45
  swing2Motion=Free
  twistLimitDegrees=5
 }
 {
  bone1Name=thigh_l
  bone2Name=pelvis
  swing1LimitDegrees=45
  swing1Motion=Free
  swing2LimitDegrees=45
  swing2Motion=Free
  twistLimitDegrees=45
  twistMotion=Free
 }
 {
  bone1Name=calf_l
  bone2Name=thigh_l
  swing1LimitDegrees=45
  swing1Motion=Free
  swing2LimitDegrees=45
  swing2Motion=Free
  twistLimitDegrees=45
  twistMotion=Free
 }
 {
  bone1Name=calf_twist_02_l
  bone2Name=calf_l
  swing1LimitDegrees=45
  swing1Motion=Free
  swing2LimitDegrees=45
  swing2Motion=Free
  twistLimitDegrees=45
  twistMotion=Free
 }
 {
  bone1Name=thigh_twist_01_l
  bone2Name=thigh_l
  swing1LimitDegrees=45
  swing1Motion=Free
  swing2LimitDegrees=45
  swing2Motion=Free
  twistLimitDegrees=45
  twistMotion=Free
 }
 {
  bone1Name=calf_r
  bone2Name=thigh_r
  swing1LimitDegrees=45
  swing1Motion=Free
  swing2LimitDegrees=45
  swing2Motion=Free
  twistLimitDegrees=45
  twistMotion=Free
 }
 {
  bone1Name=thigh_r
  bone2Name=pelvis
  swing1LimitDegrees=45
  swing1Motion=Free
  swing2LimitDegrees=45
  swing2Motion=Free
  twistLimitDegrees=45
  twistMotion=Free
 }
 {
  bone1Name=calf_twist_02_r
  bone2Name=calf_r
  swing1LimitDegrees=45
  swing1Motion=Free
  swing2LimitDegrees=45
  swing2Motion=Free
  twistLimitDegrees=45
  twistMotion=Free
 }
 {
  bone1Name=foot_r
  bone2Name=calf_r
  swing1LimitDegrees=45
  swing1Motion=Free
  swing2LimitDegrees=45
  swing2Motion=Free
  twistLimitDegrees=45
  twistMotion=Free
 }
 {
  bone1Name=ball_r
  bone2Name=foot_r
  swing1LimitDegrees=45
  swing1Motion=Free
  swing2LimitDegrees=45
  swing2Motion=Free
  twistLimitDegrees=45
  twistMotion=Free
 }
 {
  bone1Name=thigh_twist_01_r
  bone2Name=thigh_r
  swing1LimitDegrees=45
  swing1Motion=Free
  swing2LimitDegrees=45
  swing2Motion=Free
  twistLimitDegrees=45
  twistMotion=Free
 }
 {
  bone1Name=dyn_shoeLace_r_a_01
  bone2Name=dyn_shoeLace_root_r
  swing1LimitDegrees=45
  swing1Motion=Free
  swing2LimitDegrees=45
  swing2Motion=Free
  twistLimitDegrees=5
 }
 {
  bone1Name=dyn_shoeLace_r_b_01
  bone2Name=dyn_shoeLace_root_r
  swing1LimitDegrees=45
  swing1Motion=Free
  swing2LimitDegrees=45
  swing2Motion=Free
  twistLimitDegrees=5
 }
 {
  bone1Name=dyn_shoeLace_root_r
  bone2Name=foot_r
  swing1LimitDegrees=45
  swing1Motion=Free
  swing2LimitDegrees=45
  swing2Motion=Free
  twistLimitDegrees=45
  twistMotion=Free
 }
]
