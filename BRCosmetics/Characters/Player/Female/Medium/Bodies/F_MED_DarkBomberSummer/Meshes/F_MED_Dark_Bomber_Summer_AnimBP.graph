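# Post-process-style cosmetic AnimBP for the female medium Dark Bomber Summer body, copying its attached parent's pose and adding secondary motion.
# Neck collar is constrained and scaled in two identical targeting branches, then rigid-body foot physics and weighted shoelace constraints follow.
# Crouch offsets the left half-thigh bone; jumping or falling scales the right knee.
# EventGraph derives chest/strap push-in alphas from right upper-arm yaw, but those bone nodes never reach Output Pose.
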AnimBlueprint /BRCosmetics/Characters/Player/Female/Medium/Bodies/F_MED_DarkBomberSummer/Meshes/F_MED_Dark_Bomber_Summer_AnimBP
target_skeleton=&/F_MED_Dark_Bomber_Summer_Skeleton
graphs [
 AnimGraph {nodes [
   n1 = "Transform (Modify) Bone - Bone: dyn_ChestPushIn_r"(Alpha=Get ChestPushIn_r() -> (ChestPushIn_r:double), Translation="2,0,0", Rotation="0,0,0", Scale="1,1,1") {
    rotationMode=BMM_Additive
    scaleMode=BMM_Additive
    translationMode=BMM_Additive
    alphaBoolBlend {blendOption=Linear}
    boneToModify {boneName=dyn_ChestPushIn_r}
    translation {x=2}
   }
   n2 = "Transform (Modify) Bone - Bone: dyn_StrapPushIn_r"(ComponentPose=@n1, Alpha=Get StrapPushIn_r() -> (StrapPushIn_r:double), Translation="1,4,0", Rotation="0,30,0", Scale="1,1,1") {
    rotationMode=BMM_Additive
    scaleMode=BMM_Additive
    translationMode=BMM_Additive
    alphaBoolBlend {blendOption=Linear}
    boneToModify {boneName=dyn_StrapPushIn_r}
    rotation {yaw=30}
    translation=<1 4>
   }
   n3 = "Transform (Modify) Bone - Bone: dyn_ChestPushIn_r"(ComponentPose=@n2, Alpha=Get ChestPushIn_r_Corr() -> (ChestPushIn_r_Corr:double), Translation="-2,0,0", Rotation="0,0,0", Scale="1,1,1") {
    rotationMode=BMM_Additive
    scaleMode=BMM_Additive
    translationMode=BMM_Additive
    alphaBoolBlend {blendOption=Linear}
    boneToModify {boneName=dyn_ChestPushIn_r}
    translation {x=-2}
   }
   "Transform (Modify) Bone - Bone: dyn_StrapPushIn_r"(ComponentPose=@n3, Alpha=Get StrapPushIn_r_Corr() -> (StrapPushIn_r_Corr:double), Translation="-1,-4,0", Rotation="0,-30,0", Scale="1,1,1") {
    rotationMode=BMM_Additive
    scaleMode=BMM_Additive
    translationMode=BMM_Additive
    alphaBoolBlend {blendOption=Linear}
    boneToModify {boneName=dyn_StrapPushIn_r}
    rotation {yaw=-30}
    translation=<-1 -4>
   }
   n4 = "Constraint - dyn_neckCollar (P)"(
    ComponentPose=Local To Component(LocalPose=Copy Pose From Mesh() {bCopyCurves=true, bUseAttachedParent=true})
    Alpha=1
    ConstraintWeights_0=1
   ) {
    alphaBoolBlend {blendOption=Linear}
    boneToModify {boneName=dyn_neckCollar}
    constraintSetup [{targetBone {boneName=neck_02}}]
    constraintWeights [1]
   }
   n5 = "Transform (Modify) Bone - Bone: dyn_neckCollar"(ComponentPose=@n4, Alpha=1, Translation="0,0,0", Rotation="0,0,0", Scale="1.100,1.2,1.100") {
    rotationMode=BMM_Additive
    scaleMode=BMM_Additive
    translationMode=BMM_Additive
    alphaBoolBlend {blendOption=Linear}
    boneToModify {boneName=dyn_neckCollar}
    scale=<1.1 1.2 1.1>
   }
   n6 = Component To Local(ComponentPose=@n5)
   n7 = "Constraint - dyn_neckCollar (P)"(
    ComponentPose=Local To Component(LocalPose=Copy Pose From Mesh() {bCopyCurves=true, bUseAttachedParent=true})
    Alpha=1
    ConstraintWeights_0=1
   ) {
    alphaBoolBlend {blendOption=Linear}
    boneToModify {boneName=dyn_neckCollar}
    constraintSetup [{targetBone {boneName=neck_02}}]
    constraintWeights [1]
   }
   n8 = "Transform (Modify) Bone - Bone: dyn_neckCollar"(ComponentPose=@n7, Alpha=1, Translation="0,0,0", Rotation="0,0,0", Scale="1.100,1.2,1.100") {
    rotationMode=BMM_Additive
    scaleMode=BMM_Additive
    translationMode=BMM_Additive
    alphaBoolBlend {blendOption=Linear}
    boneToModify {boneName=dyn_neckCollar}
    scale=<1.1 1.2 1.1>
   }
   n9 = Component To Local(ComponentPose=@n8)
   n10 = Get bIsTargeting()
   n11 = RigidBody(ComponentPose=Local To Component(LocalPose=Blend Poses by bool(BlendPose_0=@n9, BlendPose_1=@n6, bActiveValue=@n10) {blendType=Linear})) {
    bClampLinearTranslationLimitToRefPose=true
    bOverrideWorldGravity=true
    overridePhysicsAsset=&/F_MED_Dark_Bomber_Summer_lFoot_Physics
    simulationSpace=BaseBoneSpace
    alphaBoolBlend {blendOption=Linear}
    baseBoneRef {boneName=root}
    componentAppliedLinearAccClamp=<500 500 500>
    componentLinearAccScale=<0.1 0.1 0.1>
    overrideWorldGravity {z=-980}
   }
   n12 = RigidBody(ComponentPose=@n11) {
    bClampLinearTranslationLimitToRefPose=true
    bOverrideWorldGravity=true
    overridePhysicsAsset=&/F_MED_Dark_Bomber_Summer_rFoot_Physics
    simulationSpace=BaseBoneSpace
    alphaBoolBlend {blendOption=Linear}
    baseBoneRef {boneName=root}
    componentAppliedLinearAccClamp=<500 500 500>
    componentLinearAccScale=<0.1 0.1 0.1>
    overrideWorldGravity {z=-980}
   }
   n13 = "Constraint - dyn_shoeLace_loot_l (PPPP)"!(
    ComponentPose=@n12
    Alpha=1
    ConstraintWeights_0=0.0829
    ConstraintWeights_1:float=0.5856
    ConstraintWeights_2:float=0.0282
    ConstraintWeights_3:float=0.3033
   ) -> (Pose) {
    alphaBoolBlend {blendOption=Linear}
    boneToModify {boneName=dyn_shoeLace_loot_l}
    constraintSetup [
     {targetBone {boneName=calf_l}}
     {targetBone {boneName=calf_twist_01_l}}
     {targetBone {boneName=foot_l}}
     {targetBone {boneName=calf_twist_02_l}}
    ]
    constraintWeights [1,1,1,1]
   }
   n14 = "Constraint - dyn_shoeLace_root_r (PPPP)"!(
    ComponentPose=@n13
    Alpha=1
    ConstraintWeights_0=0.2987
    ConstraintWeights_1:float=0.5969
    ConstraintWeights_2:float=0.0191
    ConstraintWeights_3:float=0.0853
   ) -> (Pose) {
    alphaBoolBlend {blendOption=Linear}
    boneToModify {boneName=dyn_shoeLace_root_r}
    constraintSetup [
     {targetBone {boneName=calf_twist_02_r}}
     {targetBone {boneName=calf_twist_01_r}}
     {targetBone {boneName=foot_r}}
     {targetBone {boneName=calf_r}}
    ]
    constraintWeights [1,1,1,1]
   }
   n15 = "Transform (Modify) Bone - Bone: thigh_l_half"(ComponentPose=@n14, Alpha=Get Crouch_Alpha() -> (Crouch_Alpha:double), Translation="0,0,-7.5", Rotation="0,0,0", Scale="1,1,1") {
    rotationMode=BMM_Additive
    rotationSpace=BCS_WorldSpace
    scaleMode=BMM_Additive
    scaleSpace=BCS_WorldSpace
    translationMode=BMM_Additive
    translationSpace=BCS_WorldSpace
    alphaBoolBlend {blendOption=Linear}
    boneToModify {boneName=thigh_l_half}
    translation {z=-7.5}
   }
   n16 = Get bIsJumping()
   n17 = Get bIsFalling()
   n18 = "Transform (Modify) Bone - Bone: dyn_r_knee"(ComponentPose=@n15, Alpha="To Float (Boolean)"(InBool=OR Boolean(A=@n16, B=@n17)), Translation="0,0,0", Rotation="0,0,0", Scale="3,1,2") {
    rotationMode=BMM_Additive
    scaleMode=BMM_Additive
    translationMode=BMM_Additive
    alphaBoolBlend {blendOption=Linear}
    boneToModify {boneName=dyn_r_knee}
    scale {x=3, z=2}
   }
   Output Pose(Result=Component To Local(ComponentPose=@n18)) {layerGroup=None}
  ]}
 EventGraph {nodes [
   Event BlueprintUpdateAnimation()
   > n1 = Sequence()
   > Set Crouch_Alpha(Crouch_Alpha:double=SelectFloat(A=1, B=0, bPickA=Get bIsCrouching())) -> (Output_Get:double)
   n2 = GetOwnerSkeletalMeshComponent()
   n3 = BreakTransform(InTransform=GetDeltaTransformFromRefPose(self=@n2, BoneName=upperarm_r, BaseName=clavicle_r))
   Set StrapPushIn_r(execute=@n1.then_1, StrapPushIn_r:double=MapRangeClamped(Value=BreakRotator(InRot=@n3.Rotation).Yaw, InRangeA=20, InRangeB=84)) -> (Output_Get:double)
   n4 = BreakTransform(InTransform=GetDeltaTransformFromRefPose(self=@n2, BoneName=upperarm_r, BaseName=clavicle_r))
   > Set StrapPushIn_r_Corr(StrapPushIn_r_Corr:double=MapRangeClamped(Value=BreakRotator(InRot=@n4.Rotation).Yaw, InRangeA=70, InRangeB=114)) -> (Output_Get:double)
   n5 = BreakTransform(InTransform=GetDeltaTransformFromRefPose(self=@n2, BoneName=upperarm_r, BaseName=clavicle_r))
   > Set ChestPushIn_r(ChestPushIn_r:double=MapRangeClamped(Value=BreakRotator(InRot=@n5.Rotation).Yaw, InRangeA=20, InRangeB=84)) -> (Output_Get:double)
   n6 = BreakTransform(InTransform=GetDeltaTransformFromRefPose(self=@n2, BoneName=upperarm_r, BaseName=clavicle_r))
   > Set ChestPushIn_r_Corr(ChestPushIn_r_Corr:double=MapRangeClamped(Value=BreakRotator(InRot=@n6.Rotation).Yaw, InRangeA=80, InRangeB=107)) -> (Output_Get:double)
   TryGetPawnOwner()
  ]}
]
variables [
 ChestPushIn_r {type=number}
 ChestPushIn_r_Corr {type=number}
 Crouch_Alpha {type=number}
 LookDown {type=number}
 StrapPushIn_r {type=number}
 StrapPushIn_r_Corr {type=number}
]
