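# Physics asset for the earrings part of the female medium Dark Bomber Summer character body, giving the earrings physical dangle.
# Box bodies on the left and right dynamic earring bones are each constrained to the head bone.
# The head carries a capsule body.
# Both joints share the same angular limits, allowing wide swing with a narrower twist.
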PhysicsAsset /BRCosmetics/Characters/Player/Female/Medium/Bodies/F_MED_DarkBomberSummer/Meshes/Parts/F_MED_Dark_Bomber_Summer_Earings_Physics
bodies {
 dyn_earing_l [{shapeType=Box, extentX=5.0856, extentY=2.7343, extentZ=0.9358, center=<2.5184 0 0>}]
 dyn_earing_r [{shapeType=Box, extentX=5.0856, extentY=2.7343, extentZ=0.7992, center=<2.5184 0 0>}]
 head [{shapeType=Capsule, length=5.1113, radius=5.2093, center=<-1.2175 2.0174 0>, rotation=<rot 90 0 58.8895>}]
}
constraints [
 {bone1Name=dyn_earing_l, bone2Name=head, swing1LimitDegrees=80, swing2LimitDegrees=80, twistLimitDegrees=30}
 {bone1Name=dyn_earing_r, bone2Name=head, swing1LimitDegrees=80, swing2LimitDegrees=80, twistLimitDegrees=30}
]
