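# Physics asset for the hair of the female medium DarkBomberSummer player cosmetic, defining bodies and joints for its ponytail and bangs.
# Capsules on the upper spine, clavicles and neck, plus capsules and boxes on the head, form the body shapes, joined by free constraints.
# Limited joints chain capsule ponytail bones, forking after 04, and box-shaped left, front and right bangs to the head.
# Front bangs lock one swing axis below their root; right bangs are stiffest.
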PhysicsAsset /BRCosmetics/Characters/Player/Female/Medium/Bodies/F_MED_DarkBomberSummer/Meshes/Parts/F_MED_Dark_Bomber_Summer_Hair_Physics
bodies {
 clavicle_l [{shapeType=Capsule, length=9.6005, radius=5.6309, center=<4.017 -1.0655 0.9399>, rotation=<rot -0.3512 -85.5878 101.2856>}]
 clavicle_r [{shapeType=Capsule, length=9.6005, radius=5.6646, center=<-4.2845 1.0199 -1.0973>, rotation=<rot -0.3512 -85.5873 101.2859>}]
 dyn_bang_front_a_01 [{shapeType=Box, extentX=1.901, extentY=1.6845, extentZ=0.8477, center=<0.9186 -0.0276 0.0285>, rotation=<rot 5.9853 1.1998 -2.1964>}]
 dyn_bang_front_a_02 [{shapeType=Box, extentX=1.1739, extentY=0.9847, extentZ=1.658, center=<0.7256 -0.1651 0.0024>, rotation=<rot -4.5496 -12.6216 3.0247>}]
 dyn_bang_front_b_01 [{shapeType=Box, extentX=2.4448, extentY=2.4734, extentZ=1.4163, center=<1.2253 0 0>, rotation=<rot -0.0248 0.1603 -0.0552>}]
 dyn_bang_front_b_02 [{shapeType=Box, extentX=5.2317, extentY=2.4734, extentZ=1.0983, center=<2.578 0.3347 -0.1117>, rotation=<rot -7.8875 3.414 22.7273>}]
 dyn_bang_front_b_03 [{shapeType=Box, extentX=4.1679, extentY=2.4734, extentZ=0.7255, center=<1.937 0.256 -0.0592>, rotation=<rot -4.6421 7.01 1.003>}]
 dyn_bang_front_root [{shapeType=Box, extentX=2.4448, extentY=2.4734, extentZ=1.04, center=<1.2253 0 0>}]
 dyn_bang_l_a_01 [{shapeType=Box, extentX=2.4584, extentY=2.2806, extentZ=0.7521, center=<1.1454 0.0093 0.0767>, rotation=<rot 6.6098 0.6113 6.9574>}]
 dyn_bang_l_a_02 [{shapeType=Box, extentX=3.1005, extentY=1.6711, extentZ=0.752, center=<1.3763 0.2477 0.0951>, rotation=<rot -6.0354 11.9166 8.6687>}]
 dyn_bang_l_b_01 [{shapeType=Box, extentX=3.1005, extentY=1.6711, extentZ=1.6433, center=<1.6222 -0.1991 0.0931>, rotation=<rot 2.2249 5.9001 -18.8667>}]
 dyn_bang_l_b_02 [{shapeType=Box, extentX=2.6287, extentY=1.2399, extentZ=0.8893, center=<1.2473 0.1986 0.0813>, rotation=<rot 2.5181 12.5263 0.5571>}]
 dyn_bang_l_b_03 [{shapeType=Box, extentX=1.01, extentY=1.01, extentZ=0.6094, center=<0.4798 0 0>}]
 dyn_bang_right_a_01 [{shapeType=Box, extentX=2.3166, extentY=1.4912, extentZ=0.683, center=<1.1996 -0.0444 0.1639>, rotation=<rot 5.1933 1.6561 2.2918>}]
 dyn_bang_right_a_02 [{shapeType=Box, extentX=1.9584, extentY=1.0839, extentZ=0.6193, center=<0.9748 0.1376 0.0195>, rotation=<rot -4.6421 7.0099 1.003>}]
 dyn_bang_right_b_01 [{shapeType=Box, extentX=2.7402, extentY=2.5206, extentZ=1.6083, center=<1.3337 0.0327 0.104>, rotation=<rot 1.7836 -5.9229 2.7581>}]
 dyn_bang_right_b_02 [{shapeType=Box, extentX=4.6458, extentY=2.5206, extentZ=1.568, center=<2.3791 0.0955 0.128>, rotation=<rot 1.2311 4.9468 3.045>}]
 dyn_bang_right_root [{shapeType=Box, extentX=1.9584, extentY=2.5538, extentZ=2.5842, center=<0.9748 0.1376 0.0195>, rotation=<rot 5.3743 -11.9439 -22.4097>}]
 dyn_ponytail_01 [{shapeType=Capsule, length=1.3294, radius=3.4824, center=<3.9947 -0.0001 -0.0323>, rotation=<rot -90 0.0049 -0.0049>}]
 dyn_ponytail_02 [{shapeType=Capsule, length=0.8621, radius=3.4271, center=<3.7335 0 0.2742>, rotation=<rot -90 -0.0074 0.0074>}]
 dyn_ponytail_03 [{shapeType=Capsule, length=0.3949, radius=3.2852, center=<3.3809 0 0.2742>, rotation=<rot -90 0.0049 -0.0049>}]
 dyn_ponytail_04 [{shapeType=Capsule, length=0.3346, radius=2.5434, center=<2.7749 -0.0001 -0.1669>, rotation=<rot -90 -0.0264 0.0264>}]
 dyn_ponytail_05 [{shapeType=Capsule, length=1.2754, radius=0.5569, center=<1.1258 -0.0125 -0.0114>, rotation=<rot 77.0574 -81.8303 -81.7039>}]
 dyn_ponytail_06 [{shapeType=Capsule, length=2.2476, radius=0.6457, center=<1.5971 0.182 -0.0456>, rotation=<rot 0 -85.3608 -85.901>}]
 dyn_ponytail_07 [{shapeType=Capsule, length=2.507, radius=0.8461, center=<2.1202 0.0001 -0.1643>, rotation=<rot -90 -0.0198 0.0198>}]
 dyn_ponytail_08 [{shapeType=Capsule, length=1.6576, radius=0.8263, center=<1.5667 0 0.0578>, rotation=<rot -90 0.0396 -0.0396>}]
 dyn_ponytail_root [{shapeType=Capsule, length=0.2846, radius=2.1992, center=<2.0425 0 0.2742>, rotation=<rot -90 0.0396 -0.0396>}]
 head [
  {shapeType=Capsule, length=10.3402, radius=6.0734, center=<2.6613 -0.7732 0>, rotation=<rot 90 0 16.2008>}
  {shapeType=Capsule, length=9.9968, radius=6.4168, center=<1.6924 1.7661 0>, rotation=<rot 90 0 -8.9835>}
  {shapeType=Box, extentX=6.9628, extentY=0.9068, extentZ=5.9276, center=<0.1162 19.1408 -6.9336>, rotation=<rot -5.4426 9.4289 15.954>}
  {shapeType=Box, extentX=2.2199, extentY=0.7922, extentZ=7.4841, center=<2.8458 21.231 9.3901>, rotation=<rot 0.5979 1.5577 -91.7041>}
 ]
 neck_01 [{shapeType=Capsule, length=3.3002, radius=5.2032, center=<0 -0.3116 -0.0001>}]
 neck_02 [{shapeType=Capsule, length=2.6562, radius=4.0428, center=<0 -0.3116 -0.0001>}]
 spine_05 [{shapeType=Capsule, length=10.5643, radius=8.3413, center=<0 -0.8888 -0.0001>}]
}
constraints [
 {
  bone1Name=clavicle_l
  bone2Name=spine_05
  swing1LimitDegrees=45
  swing1Motion=Free
  swing2LimitDegrees=45
  swing2Motion=Free
  twistLimitDegrees=45
  twistMotion=Free
 }
 {
  bone1Name=clavicle_r
  bone2Name=spine_05
  swing1LimitDegrees=45
  swing1Motion=Free
  swing2LimitDegrees=45
  swing2Motion=Free
  twistLimitDegrees=45
  twistMotion=Free
 }
 {
  bone1Name=head
  bone2Name=neck_02
  swing1LimitDegrees=45
  swing1Motion=Free
  swing2LimitDegrees=45
  swing2Motion=Free
  twistLimitDegrees=45
  twistMotion=Free
 }
 {
  bone1Name=neck_01
  bone2Name=spine_05
  swing1LimitDegrees=45
  swing1Motion=Free
  swing2LimitDegrees=45
  swing2Motion=Free
  twistLimitDegrees=45
  twistMotion=Free
 }
 {
  bone1Name=neck_02
  bone2Name=neck_01
  swing1LimitDegrees=45
  swing1Motion=Free
  swing2LimitDegrees=45
  swing2Motion=Free
  twistLimitDegrees=45
  twistMotion=Free
 }
 {bone1Name=dyn_ponytail_root, bone2Name=head, swing1LimitDegrees=20, swing2LimitDegrees=20, twistLimitDegrees=20}
 {bone1Name=dyn_ponytail_01, bone2Name=dyn_ponytail_root, swing1LimitDegrees=45, swing2LimitDegrees=45, twistLimitDegrees=10}
 {bone1Name=dyn_ponytail_02, bone2Name=dyn_ponytail_01, swing1LimitDegrees=45, swing2LimitDegrees=45, twistLimitDegrees=10}
 {bone1Name=dyn_ponytail_03, bone2Name=dyn_ponytail_02, swing1LimitDegrees=45, swing2LimitDegrees=45, twistLimitDegrees=10}
 {bone1Name=dyn_ponytail_04, bone2Name=dyn_ponytail_03, swing1LimitDegrees=45, swing2LimitDegrees=45, twistLimitDegrees=10}
 {bone1Name=dyn_ponytail_07, bone2Name=dyn_ponytail_04, swing1LimitDegrees=45, swing2LimitDegrees=45, twistLimitDegrees=10}
 {bone1Name=dyn_ponytail_08, bone2Name=dyn_ponytail_07, swing1LimitDegrees=45, swing2LimitDegrees=45, twistLimitDegrees=10}
 {bone1Name=dyn_ponytail_05, bone2Name=dyn_ponytail_04, swing1LimitDegrees=45, swing2LimitDegrees=45, twistLimitDegrees=10}
 {bone1Name=dyn_ponytail_06, bone2Name=dyn_ponytail_05, swing1LimitDegrees=45, swing2LimitDegrees=45, twistLimitDegrees=10}
 {bone1Name=dyn_bang_l_a_01, bone2Name=head, swing1LimitDegrees=45, swing2LimitDegrees=45, twistLimitDegrees=5}
 {bone1Name=dyn_bang_l_a_02, bone2Name=dyn_bang_l_a_01, swing1LimitDegrees=45, swing2LimitDegrees=45, twistLimitDegrees=5}
 {bone1Name=dyn_bang_l_b_01, bone2Name=head, swing1LimitDegrees=45, swing2LimitDegrees=45, twistLimitDegrees=5}
 {bone1Name=dyn_bang_l_b_02, bone2Name=dyn_bang_l_b_01, swing1LimitDegrees=45, swing2LimitDegrees=45, twistLimitDegrees=5}
 {bone1Name=dyn_bang_l_b_03, bone2Name=dyn_bang_l_b_01, swing1LimitDegrees=45, swing2LimitDegrees=45, twistLimitDegrees=5}
 {
  bone1Name=dyn_bang_front_a_01
  bone2Name=dyn_bang_front_root
  swing1LimitDegrees=15
  swing1Motion=Locked
  swing2LimitDegrees=15
  twistLimitDegrees=15
 }
 {
  bone1Name=dyn_bang_front_a_02
  bone2Name=dyn_bang_front_a_01
  swing1LimitDegrees=15
  swing1Motion=Locked
  swing2LimitDegrees=15
  twistLimitDegrees=15
 }
 {
  bone1Name=dyn_bang_front_b_01
  bone2Name=dyn_bang_front_root
  swing1LimitDegrees=15
  swing1Motion=Locked
  swing2LimitDegrees=15
  twistLimitDegrees=15
 }
 {
  bone1Name=dyn_bang_front_b_02
  bone2Name=dyn_bang_front_b_01
  swing1LimitDegrees=15
  swing1Motion=Locked
  swing2LimitDegrees=15
  twistLimitDegrees=15
 }
 {
  bone1Name=dyn_bang_front_b_03
  bone2Name=dyn_bang_front_b_02
  swing1LimitDegrees=15
  swing1Motion=Locked
  swing2LimitDegrees=15
  twistLimitDegrees=15
 }
 {bone1Name=dyn_bang_right_root, bone2Name=head, swing1LimitDegrees=25, swing2LimitDegrees=10, twistLimitDegrees=4}
 {bone1Name=dyn_bang_right_a_01, bone2Name=dyn_bang_right_root, swing1LimitDegrees=10, swing2LimitDegrees=1, twistLimitDegrees=2}
 {
  bone1Name=dyn_bang_right_a_02
  bone2Name=dyn_bang_right_a_01
  swing1LimitDegrees=10
  swing2LimitDegrees=1
  swing2Motion=Locked
  twistLimitDegrees=2
  twistMotion=Locked
 }
 {bone1Name=dyn_bang_front_root, bone2Name=head, swing1LimitDegrees=15, swing2LimitDegrees=15, twistLimitDegrees=15}
 {bone1Name=dyn_bang_right_b_01, bone2Name=dyn_bang_right_root, swing1LimitDegrees=5, swing2LimitDegrees=5, twistLimitDegrees=2}
 {bone1Name=dyn_bang_right_b_02, bone2Name=dyn_bang_right_b_01, swing1LimitDegrees=5, swing2LimitDegrees=5, twistLimitDegrees=2}
]
