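# Post-pose secondary-motion AnimBlueprint for the face accessory mesh of the female Dark Bomber Summer cosmetic, animating earrings and hair.
# It copies pose and curves from the attached parent mesh, then chains RigidBody simulations for earrings and for hair, the latter in root-bone space.
# An additive transform adjusts the right bang root, then low-alpha, spring-driven AnimDynamics nodes sway the top and left poof bones.
# The EventGraph holds only unconnected stub nodes.
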AnimBlueprint /BRCosmetics/Characters/Player/Female/Medium/Bodies/F_MED_DarkBomberSummer/Meshes/Parts/F_MED_Dark_Bomber_Summer_FaceAcc_AnimBP
target_skeleton=&/F_MED_Dark_Bomber_Summer_FaceAcc_Skeleton
graphs [
 AnimGraph {nodes [
   n1 = RigidBody(ComponentPose=Local To Component(LocalPose=Copy Pose From Mesh() {bCopyCurves=true, bUseAttachedParent=true})) {
    bClampLinearTranslationLimitToRefPose=true
    bOverrideWorldGravity=true
    overridePhysicsAsset=&/F_MED_Dark_Bomber_Summer_Earings_Physics
    alphaBoolBlend {blendOption=Linear}
    componentAppliedLinearAccClamp=<2500 2500 2500>
    componentLinearAccScale=<0.01 0.01 0.01>
    overrideWorldGravity {z=-1100}
   }
   n2 = RigidBody(ComponentPose=@n1) {
    bClampLinearTranslationLimitToRefPose=true
    bOverrideWorldGravity=true
    overridePhysicsAsset=&/F_MED_Dark_Bomber_Summer_Hair_Physics
    simulationSpace=BaseBoneSpace
    alphaBoolBlend {blendOption=Linear}
    baseBoneRef {boneName=root}
    componentAppliedLinearAccClamp=<3000 3000 3000>
    componentLinearAccScale=<0.15 0.2 0.6>
    overrideWorldGravity {z=-980}
   }
   n3 = "Transform (Modify) Bone - Bone: dyn_bang_right_root"(ComponentPose=@n2, Alpha=1, Translation="0,0,0", Rotation="2,0,3", Scale="0.8,0.8,0.8") {
    rotationMode=BMM_Additive
    scaleMode=BMM_Additive
    translationMode=BMM_Additive
    alphaBoolBlend {blendOption=Linear}
    boneToModify {boneName=dyn_bang_right_root}
    rotation {pitch=2, roll=3}
    scale=<0.8 0.8 0.8>
   }
   n4 = Anim Dynamics - Bone: dyn_poof_top(ComponentPose=@n3, Alpha=0.12, ExternalForce="0,0,0") {
    angularDampingOverride=0.9999
    angularSpringConstant=250
    bAngularSpring=true
    bGravityOverrideInSimSpace=true
    bLinearSpring=true
    bOverrideAngularDamping=true
    bOverrideLinearDamping=true
    bUseGravityOverride=true
    linearDampingOverride=0.9999
    linearSpringConstant=250
    simulationSpace=RootRelative
    alphaBoolBlend {blendOption=Linear}
    boundBone {boneName=dyn_poof_top}
    componentLinearAccScale=<0.25 0.25 0.25>
    gravityOverride {z=-1100}
    physicsBodyDefinitions [{
      sphereCollisionRadius=0
      boundBone {boneName=dyn_poof_top}
      boxExtents=<10 15 15>
      constraintSetup {angularLimitsMax=<25 25 25>, angularLimitsMin=<-25 -25 -25>, angularTarget=<0 0 0>, linearAxesMax=<2 2 2>, linearAxesMin=<-2 -2 -2>}
      localJointOffset=<0 -0.5 -1>
     }]
    relativeSpaceBone {boneName=root}
   }
   n5 = Anim Dynamics - Bone: dyn_poof_l(ComponentPose=@n4, Alpha=0.2, ExternalForce="0,0,0") {
    angularDampingOverride=0.9999
    angularSpringConstant=250
    bAngularSpring=true
    bGravityOverrideInSimSpace=true
    bLinearSpring=true
    bOverrideAngularDamping=true
    bOverrideLinearDamping=true
    bUseGravityOverride=true
    linearDampingOverride=0.9999
    linearSpringConstant=250
    simulationSpace=RootRelative
    alphaBoolBlend {blendOption=Linear}
    boundBone {boneName=dyn_poof_l}
    componentLinearAccScale=<0.25 0.25 0.25>
    gravityOverride {z=-1100}
    physicsBodyDefinitions [{
      sphereCollisionRadius=0
      boundBone {boneName=dyn_poof_l}
      boxExtents=<10 15 15>
      constraintSetup {angularLimitsMax=<25 25 25>, angularLimitsMin=<-25 -25 -25>, angularTarget=<0 0 0>, linearAxesMax=<2 2 2>, linearAxesMin=<-2 -2 -2>}
      localJointOffset=<0 -0.5 -1>
     }]
    relativeSpaceBone {boneName=root}
   }
   Output Pose(Result=Component To Local(ComponentPose=@n5)) {layerGroup=None}
  ]}
 EventGraph {nodes [
   TryGetPawnOwner()
   Event BlueprintUpdateAnimation()
  ]}
]
variables []
